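# Parent character material for the CreamDecoy Jam skin, inserting stylized rim lighting and highlights between the shared 2023 CharacterParent PreFX and PostFX functions.
# View-space direction vectors dotted with normals, plus Fresnel, create camera-relative rim, highlight and shading terms, with masks and ambient occlusion controls.
# The rim and highlight result blends a tinted base color and emissive over the material, with near/far tuning and camera depth fade.
# A static-switch "white variant fix" adjusts base color.
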Material /BRCosmetics/Characters/Player/Female/Medium/Bodies/F_MED_CreamDecoy/Skins/Jam/Materials/M_CreamDecoy_Parent
graph {
 properties {
  bAutomaticallySetUsageInEditor=false
  bUseMaterialAttributes=true
  bUsedWithMeshDeformer=true
  bUsedWithMorphTargets=true
  bUsedWithSkeletalMesh=true
  pixelDepthOffsetMode=PDOM_Legacy
  shadingModel=MSM_FromMaterialExpression
  usedShadingModels=Default Lit | Subsurface | Clear Coat | Subsurface Profile | Cloth
 }
 nodes [
  n1 = MaterialFunctionCall() {materialFunction=&MF_CharacterParent_2023_PreFX}
  MaterialFunctionCall(Material=NamedRerouteUsage(), M=@n1.M, Pre Skinned Local Position=@n1.M, Distance Blend=@n1.M, Fresnel=@n1.M) {materialFunction=&MF_CharacterParent_2023_PostFX}
  n2 = GetMaterialAttributes(None=NamedRerouteUsage()) {attributeGetTypes [#1]}
  n3 = Multiply(A=@n2.Base Color, B=NamedRerouteUsage())
  n4 = LinearInterpolate(
   A=@n3
   B=Multiply(
    A=@n3
    B=ScalarParameter() {defaultValue=0.85, group="CreamDecoy - Rimlights && Highlights", parameterName=Distance - Base Color Low End Multiplier}
   ) {constB=0.8}
   Alpha=NamedRerouteUsage()
  )
  MaterialFunctionCall!(Low=@n4, High=@n3, Epic=@n3, --------, Mobile=@n4, Medium=@n4) {materialFunction=/FortniteBaseMaterials/Functions/MF_QualitySwitch_Vector}
  n5 = VectorParameter() {group="CreamDecoy - Rimlights && Highlights", parameterName=RimLightColor, defaultValue=<rgba 0.1379 0.1735 0.3021 1>}
  n6 = ScalarParameter() {group="CreamDecoy - Rimlights && Highlights", parameterName=BaseColorContribution}
  Multiply(A=@n5.RGB, B=@n6)
  NamedRerouteDeclaration(None=@n1.Result) {name=Material In, nodeColor=<rgba 1 0 0.8906 1>}
  n7 = StaticSwitchParameter(
   True=SetMaterialAttributes!(MaterialAttributes=@n2.MaterialAttributes, Base Color=@n4) {attributeSetTypes [#1]}
   False=NamedRerouteUsage()
  ) {group=CreamDecoy, parameterName=UseWhiteVariantFix}
  n8 = GetMaterialAttributes(None=@n7) {attributeGetTypes [#1, #2]}
  n9 = LinearInterpolate(A=@n8.Base Color, B=@n5.RGB, Alpha=@n6)
  n10 = Add(A=@n8.Emissive Color, B=@n5.RGB)
  n11 = VectorParameter() {group="CreamDecoy - Rimlights && Highlights", parameterName=RimLightScreenSpacePosition, defaultValue=<rgba 1 0.5 1 1>}
  n12 = DotProduct(A=Transform(None=@n11.RGB) {transformSourceType=TRANSFORMSOURCE_View}, B=PixelNormalWS())
  n13 = Fresnel(ExponentIn=LinearInterpolate(
    A=ScalarParameter() {defaultValue=1, group="CreamDecoy - Rimlights && Highlights", parameterName=FresnelBaseExponent - Near}
    B=ScalarParameter() {defaultValue=1, group="CreamDecoy - Rimlights && Highlights", parameterName=FresnelBaseExponent - Far}
    Alpha=NamedRerouteUsage()
   )) {baseReflectFraction=0, exponent=1}
  n14 = SmoothStep(
   Min=ScalarParameter() {group="CreamDecoy - Rimlights && Highlights", parameterName=RimLight - Min}
   Max=ScalarParameter() {defaultValue=1, group="CreamDecoy - Rimlights && Highlights", parameterName=RimLight - Max}
   Value=Multiply(A=Saturate(None=@n12), B=@n13)
  )
  n15 = Multiply(
   A=Power(Base=@n13, Exp=ScalarParameter() {defaultValue=4, group="CreamDecoy - Rimlights && Highlights", parameterName=FresnelPow})
   B=ScalarParameter() {defaultValue=0.25, group="CreamDecoy - Rimlights && Highlights", parameterName=FresnelIntensit}
  )
  n16 = OneMinus(None=ScalarParameter() {group="CreamDecoy - Rimlights && Highlights", parameterName=RimLight - AO Intensity})
  n17 = NamedRerouteUsage()
  n18 = Multiply(A=@n17, B=@n17)
  n19 = VectorParameter() {group="CreamDecoy - Rimlights && Highlights|Highlights", parameterName=HighlightsScreenSpacePosition, defaultValue=<rgba -0.25 0 -1 1>}
  n20 = Transform(None=@n19.RGB) {transformSourceType=TRANSFORMSOURCE_View}
  n21 = LinearInterpolate(A=PixelNormalWS(), B=VertexNormalWS())
  n22 = Power(
   Base=Saturate(None=DotProduct(A=@n20, B=@n21))
   Exp=ScalarParameter() {defaultValue=32, group="CreamDecoy - Rimlights && Highlights|Highlights", parameterName=HighlightExp}
  ) {constExponent=64}
  n23 = NamedRerouteUsage()
  n24 = ScalarParameter() {defaultValue=0.5, group="CreamDecoy - Rimlights && Highlights|Highlights", parameterName=HighligtsIntensity_Near}
  n25 = ScalarParameter() {defaultValue=0.1, group="CreamDecoy - Rimlights && Highlights|Highlights", parameterName=HighligtsIntensity_Far}
  n26 = NamedRerouteUsage()
  n27 = Max(
   A=Multiply(A=Max(A=@n14, B=@n15), B=LinearInterpolate(A=@n16, Alpha=@n18))
   B=Multiply(A=Multiply(A=@n22, B=@n23), B=LinearInterpolate(A=@n24, B=@n25, Alpha=@n26))
  )
  n28 = Multiply(A=@n27, B=ScalarParameter() {defaultValue=1, group="CreamDecoy - Rimlights && Highlights", parameterName=RimLightColorIntensity - Near})
  n29 = Multiply(A=@n27, B=ScalarParameter() {defaultValue=1, group="CreamDecoy - Rimlights && Highlights", parameterName=RimLightColorIntensity - Far})
  n30 = NamedRerouteUsage()
  NamedRerouteDeclaration(None=BlendMaterialAttributes(
    A=@n7
    B=SetMaterialAttributes!(MaterialAttributes=@n8.MaterialAttributes, Base Color=@n9, Emissive Color=@n10) {attributeSetTypes [#1, #2]}
    Alpha=LinearInterpolate(A=@n28, B=@n29, Alpha=@n30)
   )) {name=Material Out, nodeColor=<rgba 0 1 0.7891 1>}
  NamedRerouteDeclaration(None=@n1.Tech Art Mask) {name=FXMask In, nodeColor=<rgba 1 0 0.5859 1>}
  n31 = ChannelMaskParameter(None=NamedRerouteUsage()) {group="CreamDecoy - Rimlights && Highlights", maskChannel=Blue, parameterName=CheckersHighlightMask, defaultValue=<rgba 0 0 1 0>}
  NamedRerouteDeclaration(None=@n31.R) {name=HighlightsMask, nodeColor=<rgba 1 0.2109 0 1>}
  NamedRerouteDeclaration(None=MaterialFunctionCall(
    Fade Length=ScalarParameter() {defaultValue=512, group="CreamDecoy - Rimlights && Highlights", parameterName="Fade Length (S)"}
    Fade Offset=ScalarParameter() {defaultValue=350, group="CreamDecoy - Rimlights && Highlights", parameterName="Fade Offset (S)"}
   ) {materialFunction=&CameraDepthFade}.Result) {name=CameraDistanceFade, nodeColor=<rgba 0.5234 1 0 1>}
  n32 = VectorParameter() {group="CreamDecoy - Rimlights && Highlights|Highlights", parameterName=ShadingScreenSpacePosition, defaultValue=<rgba -0.5 0.5 -1 1>}
  n33 = Transform(None=@n32.RGB) {transformSourceType=TRANSFORMSOURCE_View}
  n34 = PixelNormalWS()
  n35 = Saturate(None=ConstantBiasScale(None=DotProduct(A=@n33, B=@n34)))
  n36 = ScalarParameter() {defaultValue=6, parameterName=Exp}
  NamedRerouteDeclaration(None=LinearInterpolate(A=ScalarParameter() {defaultValue=0.5, parameterName=A}, Alpha=Power(Base=@n35, Exp=@n36) {constExponent=8})) {name=ScreenSpaceShading, nodeColor=<rgba 0 1 0.8125 1>}
  NamedRerouteDeclaration(None=@n1.M) {name=Mask In, nodeColor=<rgba 0 1 0.2969 1>}
  NamedRerouteDeclaration(None=MaterialFunctionCall(Color=NamedRerouteUsage()) {materialFunction=&SplitComponents}.R) {name=AmbientOcclusion, nodeColor=<rgba 1 0 0.0234 1>}
  NamedRerouteUsage()
 ]
}
